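# Editor utility Blueprint that bakes skeletal animation into textures for static mesh playback, using the AnimToTexture plugin and a bone animation data asset.
# When run, it sets the data asset's static mesh LOD to a specific lightmap index with generated lightmap UVs, then runs the AnimationToTexture bake.
# On success, it pushes the data asset's settings as global parameters into the Body and ChestLog bone animation material instances.
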EditorUtilityBlueprint /ProtoSwarm/AnimToTextures/BP_AnimToTexture
graphs [EventGraph {nodes [
   Event Run()
   n1 = Get BoneDataAsset() -> (BoneDataAsset:Anim To Texture Data Asset Object Reference)
   > SetLightMapIndex!(
    execute
    StaticMesh:Static Mesh Object Reference=Get Static Mesh!(self:Anim To Texture Data Asset Object Reference=@n1) -> (ReturnValue:Static Mesh Object Reference) {is=title K2Node_CallFunction}
    LODIndex:int=Get StaticLODIndex!(self:Anim To Texture Data Asset Object Reference=@n1) -> (StaticLODIndex:int=0)
    LightmapIndex:int=2
    bGenerateLightmapUVs:bool=true
   ) -> (then, ReturnValue:bool)
   > n2 = AnimationToTexture!(execute, DataAsset:Anim To Texture Data Asset Object Reference=Get BoneDataAsset()) -> (then, ReturnValue:bool)
   > Branch(Condition=@n2)
   > UpdateMaterialInstanceFromDataAsset!(
    execute
    DataAsset:Anim To Texture Data Asset Object Reference=Get BoneDataAsset()
    MaterialInstance:Material Instance Constant Object Reference=/ProtoSwarm/AnimToTextures/Materials/BoneAnimation/MI_Body_BoneAnimation
    MaterialParameterAssociation:EMaterialParameterAssociation Enum=GlobalParameter
   ) -> (then)
   > UpdateMaterialInstanceFromDataAsset(
    DataAsset=Get BoneDataAsset()
    MaterialInstance=/ProtoSwarm/AnimToTextures/Materials/BoneAnimation/MI_ChestLog_BoneAnimation
    MaterialParameterAssociation=GlobalParameter
   )
  ]}]
variables [BoneDataAsset {type=AnimToTextureDataAsset, default=/ProtoSwarm/AnimToTextures/Data/DA_BoneAnimation}]
