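; Blueprint actor that feeds the instigator's location into the TidalCrane boss target Niagara effect, apparently for the Lobber projectile attack.
; On BeginPlay it writes the instigator's actor location into the effect's BomberLocation position variable, once, with no later updates.
; The overlap and tick events are empty, as is the construction script.
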(event EventBeginPlay
  (Niagara|SetNiagaraVariable(Position) (Variables|Default|GetNS_TidalCrane_BossTarget) "BomberLocation" (Transformation|GetActorLocation (Game|GetInstigator))))

(event Collision|EventActorBeginOverlap
  (params (OtherActor :type Actor%20Object%20Reference)))

(event EventTick
  (params (DeltaSeconds :type Float%20%28single-precision%29)))

(fn ConstructionScript)
