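# Material function for texture-based bone animation that outputs a position and normal blended from one, two or four bone influences at a given frame.
# UnpackBoneWeights supplies indices and weights, and BonePositionAndNormal samples position and rotation textures per bone, using bounding box inputs.
# Nested static switches select the influence count.
# As wired, both blend functions receive positions as normals and the first weight for every influence.
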MaterialFunction /AnimToTexture/Materials/MaterialFunctions/BlendBonePositionAndNormal
graph {nodes [
  n1 = NamedRerouteUsage()
  n2 = NamedRerouteUsage()
  n3 = FunctionInput() {inputName=UseTwoInfluences, inputType=FunctionInput_StaticBool, sortPriority=8}
  n4 = FunctionInput() {inputName=UseFourInfluences, inputType=FunctionInput_StaticBool, sortPriority=9}
  FunctionOutput(None=StaticSwitch(True=NamedRerouteUsage(), False=StaticSwitch(True=@n1, False=@n2, Value=@n3), Value=@n4)) {outputName=Normal}
  n5 = NamedRerouteUsage()
  n6 = NamedRerouteUsage()
  FunctionOutput(None=StaticSwitch(True=NamedRerouteUsage(), False=StaticSwitch(True=@n5, False=@n6, Value=@n3), Value=@n4)) {outputName=Position, sortPriority=1}
  n7 = FunctionInput() {inputName=Frame, inputType=FunctionInput_Scalar}
  n8 = MaterialFunctionCall!(
   Tex=FunctionInput() {inputName=BoneWeightsTexture, inputType=FunctionInput_Texture2D, sortPriority=3}
   RowsPerFrame=FunctionInput() {inputName=BoneWeightsRowsPerFrame, inputType=FunctionInput_Scalar, sortPriority=4}
   NumBones=FunctionInput() {inputName=NumBones, inputType=FunctionInput_Scalar, sortPriority=5}
  ) {materialFunction=&/UnpackBoneWeights}
  n9 = MaterialFunctionCall(Float4=@n8.Indices) {materialFunction=&BreakOutFloat4Components}
  n10 = FunctionInput() {inputName=BonePositionTexture, inputType=FunctionInput_Texture2D, sortPriority=1}
  n11 = FunctionInput() {inputName=BoneRotationTexture, inputType=FunctionInput_Texture2D, sortPriority=2}
  n12 = FunctionInput() {inputName=SizeBBox, sortPriority=6}
  n13 = FunctionInput() {inputName=MinBBox, sortPriority=7}
  n14 = MaterialFunctionCall!(Frame=@n7, BoneIndex=@n9.R, BonePositionTexture=@n10, BoneRotationTexture=@n11, SizeBBox=@n12, MinBBox=@n13) {desc=First Influence, materialFunction=&/BonePositionAndNormal}
  NamedRerouteDeclaration(None=@n14.Normal) {name=BlendedOneNormal, nodeColor=<rgba 0 1 0.6953 1>}
  n15 = MaterialFunctionCall(Float4=@n8.Weights) {desc=Influence Weights, materialFunction=&BreakOutFloat4Components}
  n16 = MaterialFunctionCall(Frame=@n7, BoneIndex=@n9.G, BonePositionTexture=@n10, BoneRotationTexture=@n11, SizeBBox=@n12, MinBBox=@n13) {desc=Second Influence, materialFunction=&BonePositionAndNormal}
  n17 = MaterialFunctionCall!(Position1=@n14.Position, Normal1=@n14.Position, Weight1=@n15.R, Position2=@n16.Position, Normal2=@n16.Position, Weight2=@n15.R) {materialFunction=&/BlendTwoInfluences}
  NamedRerouteDeclaration(None=@n17.Normal) {name=BlendedTwoNormals, nodeColor=<rgba 0 1 0.6953 1>}
  n18 = MaterialFunctionCall(Frame=@n7, BoneIndex=@n9.B, BonePositionTexture=@n10, BoneRotationTexture=@n11, SizeBBox=@n12, MinBBox=@n13) {desc=Third Influence, materialFunction=&BonePositionAndNormal}
  n19 = MaterialFunctionCall(Frame=@n7, BoneIndex=@n9.A, BonePositionTexture=@n10, BoneRotationTexture=@n11, SizeBBox=@n12, MinBBox=@n13) {desc=Forth Influence, materialFunction=&BonePositionAndNormal}
  n20 = MaterialFunctionCall!(
   Position1=@n14.Position
   Normal1=@n14.Position
   Weight1=@n15.R
   Position2=@n16.Position
   Normal2=@n16.Position
   Weight2=@n15.R
   Position3=@n18.Position
   Normal3=@n18.Position
   Weight3=@n15.R
   Position4=@n19.Position
   Normal4=@n19.Position
   Weight4=@n15.R
  ) {materialFunction=&/BlendFourIfluences}
  NamedRerouteDeclaration(None=@n20.Normal) {name=BlendedFourNormals, nodeColor=<rgba 0 1 0.9297 1>}
  NamedRerouteDeclaration(None=@n14.Position) {name=BlendedOnePosition, nodeColor=<rgba 1 0 0.9609 1>}
  NamedRerouteDeclaration(None=@n17.Position) {name=BlendedTwoPositions, nodeColor=<rgba 1 0 0.9609 1>}
  NamedRerouteDeclaration(None=@n20.Position) {name=BlendedFourPositions, nodeColor=<rgba 1 0 0.9609 1>}
 ]}
