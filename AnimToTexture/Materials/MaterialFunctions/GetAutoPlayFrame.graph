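# GetAutoPlayFrame is an AnimToTexture material function that computes a looping animation frame from material time.
# It adds TimeOffset to Time, then multiplies by Playrate and SampleRate, clamping offset and rate to non-negative values.
# Fmod wraps the result over the inclusive StartFrame-to-EndFrame count, then StartFrame is added back.
# The Frame output is not rounded, so it can be fractional.
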MaterialFunction /AnimToTexture/Materials/MaterialFunctions/GetAutoPlayFrame
graph {nodes [
  n1 = FunctionInput() {inputName=TimeOffset, inputType=FunctionInput_Scalar, sortPriority=3}
  n2 = FunctionInput() {inputName=Playrate, inputType=FunctionInput_Scalar, sortPriority=4, previewValue=<1 1 1 1>}
  n3 = Multiply(
   A=Add(A=Time(), B=Max(A=@n1) {constB=0})
   B=Multiply(A=Max(A=@n2) {constB=0}, B=FunctionInput() {inputName=SampleRate, inputType=FunctionInput_Scalar, sortPriority=8, previewValue=<1 1 1 1>})
  )
  n4 = FunctionInput() {inputName=StartFrame, inputType=FunctionInput_Scalar, sortPriority=5}
  n5 = Add(A=Subtract(A=FunctionInput() {inputName=EndFrame, inputType=FunctionInput_Scalar, sortPriority=6}, B=@n4))
  FunctionOutput(None=Add(A=Fmod(A=@n3, B=@n5), B=@n4)) {outputName=Frame}
 ]}
