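# GetFrame is an AnimToTexture material function that keeps a supplied animation frame, described as custom time, within the available frame range.
# It clamps the Frame input between the default minimum and NumFrames minus the default subtract constant, then outputs it as Frame.
# It does no time-based playback, looping or rounding itself.
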MaterialFunction /AnimToTexture/Materials/MaterialFunctions/GetFrame
graph {nodes [
  n1 = FunctionInput() {inputName=NumFrames, inputType=FunctionInput_Scalar, sortPriority=7}
  FunctionOutput(None=Clamp(None=FunctionInput() {desc=Custom Time, inputName=Frame, inputType=FunctionInput_Scalar}, Max=Subtract(A=@n1))) {outputName=Frame}
 ]}
