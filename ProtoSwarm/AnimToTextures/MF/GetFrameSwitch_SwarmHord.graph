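# Material function that selects the animation frame for AnimToTexture vertex animation, apparently for ProtoSwarm's swarm horde.
# Static switch AutoPlay chooses between time-driven GetAutoPlayFrame and explicitly supplied frames through GetFrame.
# In both modes, UseDynamicParameters selects Dynamic Parameters or per-instance custom data with scalar parameter defaults as the input source.
# The manual path uses PreviousFrameSwitch, with separate Frame and PrevFrame values in the per-instance branch.
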MaterialFunction /ProtoSwarm/AnimToTextures/MF/GetFrameSwitch_SwarmHord
graph {nodes [
  n1 = DynamicParameter() {parameterIndex=1, paramNames [TimeOffset,Playrate,StartFrame,EndFrame]}
  n2 = MaterialFunctionCall!(TimeOffset=@n1.TimeOffset, Playrate=@n1.TimeOffset, StartFrame=@n1.TimeOffset, EndFrame=@n1.TimeOffset, SampleRate=NamedRerouteUsage()) {materialFunction=/AnimToTexture/Materials/MaterialFunctions/GetAutoPlayFrame}
  n3 = MaterialFunctionCall(
   TimeOffset=PerInstanceCustomData!(DefaultValue=ScalarParameter() {group=AnimToTexture, parameterName=TimeOffset, sortPriority=0})
   Playrate=PerInstanceCustomData(DefaultValue=ScalarParameter() {defaultValue=1, group=AnimToTexture, parameterName=Playrate, sortPriority=0}) {constDefaultValue=1, dataIndex=1}
   StartFrame=PerInstanceCustomData(DefaultValue=ScalarParameter() {group=AnimToTexture, parameterName=StartFrame, sortPriority=0}) {dataIndex=2}
   EndFrame=PerInstanceCustomData(DefaultValue=ScalarParameter() {group=AnimToTexture, parameterName=EndFrame, sortPriority=0}) {dataIndex=3}
   SampleRate=NamedRerouteUsage()
  ) {materialFunction=&GetAutoPlayFrame}
  n4 = NamedRerouteUsage()
  n5 = DynamicParameter() {paramNames [Frame,PrevFrame,NumFrames,Param4]}
  n6 = MaterialFunctionCall!(Frame=PreviousFrameSwitch(Current Frame=@n5.Frame, Previous Frame=@n5.Frame), NumFrames=@n5.NumFrames) {materialFunction=/AnimToTexture/Materials/MaterialFunctions/GetFrame}
  n7 = ScalarParameter() {group=AnimToTexture, parameterName=Frame, sortPriority=0}
  n8 = ScalarParameter() {group=AnimToTexture, parameterName=PrevFrame, sortPriority=0}
  n9 = MaterialFunctionCall(
   Frame=PreviousFrameSwitch(Current Frame=PerInstanceCustomData(DefaultValue=@n7), Previous Frame=PerInstanceCustomData(DefaultValue=@n8) {dataIndex=1})
   NumFrames=NamedRerouteUsage()
  ) {materialFunction=&GetFrame}
  n10 = NamedRerouteUsage()
  FunctionOutput(None=StaticSwitch(
    True=StaticSwitch(True=@n2.Frame, False=@n3.Frame, Value=@n4)
    False=StaticSwitch(True=@n6.Frame, False=@n9.Frame, Value=@n10)
    Value=StaticBoolParameter() {defaultValue=true, group=AnimToTexture, parameterName=AutoPlay, sortPriority=10}
   ))
  NamedRerouteDeclaration(None=ScalarParameter() {defaultValue=1, group=AnimToTexture, parameterName=NumFrames, sortPriority=20}) {name=NumFrames, nodeColor=<rgba 0 1 0.2266 1>}
  NamedRerouteDeclaration(None=ScalarParameter() {defaultValue=30, group=AnimToTexture, parameterName=SampleRate, sortPriority=0}) {name=SampleRate, nodeColor=<rgba 0 0.25 1 1>}
  NamedRerouteDeclaration(None=StaticBoolParameter() {group=AnimToTexture, parameterName=UseDynamicParameters, sortPriority=10}) {name=UseDynamicParameters, nodeColor=<rgba 1 0 0.8203 1>}
  NamedRerouteUsage()
  NamedRerouteUsage()
  NamedRerouteUsage()
 ]}
